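# Animation Blueprint for the MimicGlean Awe cosmetic tail accessory, built on the shared Fortnite base tail skeleton.
# The output takes the input pose and adds a small fixed bone-space yaw to the tail base joint.
# An idle, locomotion and jump state machine reuses Neon Cat tail animations and saves to cached pose 'Entry', which the output never uses.
# The event graph has no connected logic, and RelaxSpeed and TrailAlpha are unused.
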AnimBlueprint /BRCosmetics/Accessories/FORT_Tails/Tail_M_MED_MimicGlean_Awe/Meshes/M_MED_MimicGlean_Awe_Tail_AnimBP
target_skeleton=/Game/Accessories/FORT_Tails/Common/Fortnite_Base_Tail_Skeleton
graphs [
 AnimGraph {nodes [
   n1 = Local To Component(LocalPose=Input Pose())
   n2 = Get bIsMoving2D()
   Output Pose(Result=Component To Local(ComponentPose="Transform (Modify) Bone - Bone: C_Tail_A_Base_Jnt"(ComponentPose=@n1, Alpha=1, Translation="0,0,0", Rotation="0,5,0", Scale="1,1,1", None=@n2) {rotationMode=BMM_Additive, rotationSpace=BCS_BoneSpace, boneToModify {boneName=C_Tail_A_Base_Jnt}, rotation {yaw=5}}))
   AnimGraphNode_StateMachine_0 = Statrt cache() {is=title AnimGraphNode_StateMachine}
   Save cached pose 'Entry'(Pose=@AnimGraphNode_StateMachine_0) {cacheName=Entry}
  ]}
 Statrt cache {
  entry=idle
  path=AnimGraph.AnimGraphNode_StateMachine_0.Statrt cache
  ids=id
  nodes [
   AnimStateNode_1 = Locomotion() {is=title AnimStateNode}
   AnimStateNode_2 = idle() {is=title AnimStateNode}
   AnimStateNode_4 = Jump() {is=title AnimStateNode}
   AnimStateTransitionNode_0 = AnimStateTransitionNode() {bAllowInertializationForSelfTransitions=true}
   AnimStateTransitionNode_1 = AnimStateTransitionNode() {bAllowInertializationForSelfTransitions=true}
   AnimStateTransitionNode_3 = AnimStateTransitionNode() {bAllowInertializationForSelfTransitions=true}
   AnimStateTransitionNode_4 = AnimStateTransitionNode() {bAllowInertializationForSelfTransitions=true}
  ]
  links [
   {
    dst=idle
    rule=AnimGraph.AnimGraphNode_StateMachine_0.Statrt cache.AnimStateTransitionNode_4.Transition
    src=Jump
    via=AnimStateTransitionNode_4
   }
   {
    dst=idle
    rule=AnimGraph.AnimGraphNode_StateMachine_0.Statrt cache.AnimStateTransitionNode_0.Transition
    src=Locomotion
    via=AnimStateTransitionNode_0
   }
   {
    dst=Jump
    rule=AnimGraph.AnimGraphNode_StateMachine_0.Statrt cache.AnimStateTransitionNode_3.Transition
    src=idle
    via=AnimStateTransitionNode_3
   }
   {
    dst=Locomotion
    rule=AnimGraph.AnimGraphNode_StateMachine_0.Statrt cache.AnimStateTransitionNode_1.Transition
    src=idle
    via=AnimStateTransitionNode_1
   }
  ]
 }
 Locomotion {
  path=AnimGraph.AnimGraphNode_StateMachine_0.Statrt cache.AnimStateNode_1.Locomotion
  nodes [Output Animation Pose(Result=Sequence Player 'Tail_Sprint'!(PlayRate:float=0.3) -> (Pose) {bLoopAnimation=false, sequence=/Game/Accessories/FORT_Tails/Tail_Neon_Cat/Anims/Tail_Sprint})]
 }
 idle {
  path=AnimGraph.AnimGraphNode_StateMachine_0.Statrt cache.AnimStateNode_2.idle
  nodes [Output Animation Pose(Result=Sequence Player 'Tail_Idle'() {sequence=/Game/Accessories/FORT_Tails/Tail_Neon_Cat/Anims/Tail_Idle})]
 }
 Jump {
  path=AnimGraph.AnimGraphNode_StateMachine_0.Statrt cache.AnimStateNode_4.Jump
  nodes [Output Animation Pose(Result=Sequence Player 'Tail_Jump_Loop_Apex_01'() {bLoopAnimation=false, sequence=/Game/Accessories/FORT_Tails/Tail_Neon_Cat/Anims/Tail_Jump_Loop_Apex_01})]
 }
 Transition {
  path=AnimGraph.AnimGraphNode_StateMachine_0.Statrt cache.AnimStateTransitionNode_0.Transition
  nodes [Result(bCanEnterTransition=Get bTransition_Locomotion_Idle() -> (bTransition_Locomotion_Idle:bool))]
 }
 Transition {
  path=AnimGraph.AnimGraphNode_StateMachine_0.Statrt cache.AnimStateTransitionNode_1.Transition
  nodes [Result(bCanEnterTransition=Get bTransition_Idle_Locomotion() -> (bTransition_Idle_Locomotion:bool))]
 }
 Transition {
  path=AnimGraph.AnimGraphNode_StateMachine_0.Statrt cache.AnimStateTransitionNode_3.Transition
  nodes [Result(bCanEnterTransition=Get bIsJumping())]
 }
 Transition {
  path=AnimGraph.AnimGraphNode_StateMachine_0.Statrt cache.AnimStateTransitionNode_4.Transition
  nodes [Result(bCanEnterTransition=Get bIsFalling())]
 }
 EventGraph {nodes [
   TryGetPawnOwner()
   Event BlueprintUpdateAnimation()
  ]}
]
variables [
 RelaxSpeed {type=number}
 TrailAlpha {type=number}
]
